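# Sound cue for the FruitFire NPC's Wild Estate voice lines, played, by its name, when returning to idle from suspicion.
# A Random node picks one of seven equally weighted SusReturnIdle wave players.
# It uses shared henchman attenuation, the critical enemy VO sound class and boss combat chatter concurrency.
# Bus sends feed sidechain buses for music, weapon and explosion volume.
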SoundCue /FruitFire_NudgeHonk/Assets/VO/Waves/IP_SC_VO_WildEstate_FruitFire_SusReturnIdle_Cue
root=Random_0
nodes [
 {
  id=Random_0
  children [WavePlayer_0,WavePlayer_1,WavePlayer_2,WavePlayer_3,WavePlayer_4,WavePlayer_5,WavePlayer_6]
  properties {weights [1,1,1,1,1,1,1]}
 }
 {id=WavePlayer_0, properties {soundWaveAssetPtr=&/vo_IP_WildEstate_FruitFire_SusReturnIdle_010}}
 {id=WavePlayer_1, properties {soundWaveAssetPtr=&/vo_IP_WildEstate_FruitFire_SusReturnIdle_020}}
 {id=WavePlayer_2, properties {soundWaveAssetPtr=&/vo_IP_WildEstate_FruitFire_SusReturnIdle_030}}
 {id=WavePlayer_3, properties {soundWaveAssetPtr=&/vo_IP_WildEstate_FruitFire_SusReturnIdle_040}}
 {id=WavePlayer_4, properties {soundWaveAssetPtr=&/vo_IP_WildEstate_FruitFire_SusReturnIdle_050}}
 {id=WavePlayer_5, properties {soundWaveAssetPtr=&/vo_IP_WildEstate_FruitFire_SusReturnIdle_070}}
 {id=WavePlayer_6, properties {soundWaveAssetPtr=&/vo_IP_WildEstate_FruitFire_SusReturnIdle_080}}
]
properties {
 attenuationSettings=/Game/Sounds/Fort_Human_VO_BR/NPCs/Shared/NPC_VO_Henchmen_Att
 maxDistance=8500
 soundClassObject=&Fort_VO_Enemy_Critical
 volumeMultiplier=0.8414
 busSends [
  {audioBus=&AB_Sidechain_FortMusic_Vol, minSendLevel=0, sendLevel=1, soundSourceBus=null}
  {audioBus=/Game/Sounds/Modulation/EnvelopeFollowers/Volume/AB_Sidechain_FortWeapon1P_Vol, minSendLevel=0, sendLevel=1, soundSourceBus=null}
  {audioBus=&AB_Sidechain_FortWeapon_Vol, minSendLevel=0, sendLevel=1, soundSourceBus=null}
  {audioBus=&AB_Sidechain_FortExplosion_Vol, minSendLevel=0, sendLevel=1, soundSourceBus=null}
 ]
 concurrencySet [&NPC_Boss_Combat_Chatter_Conc]
}
